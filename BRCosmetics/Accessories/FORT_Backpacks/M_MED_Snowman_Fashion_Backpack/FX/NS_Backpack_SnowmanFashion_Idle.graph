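# Niagara idle effect for the Snowman Fashion backpack, emitting small tinted snowflakes and soft, growing puffs.
# Both sprite emitters place particles along User.SkeletalMesh with partitioned BoneFlow, push them against the spline tangent, then let them drift down under drag.
# Flakes spawn at a steady rate, while puffs gate spawn probability on owner speed and play SubUV animation.
# Both fade over life and by distance from an offset owner position via a shared system curve.
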NiagaraSystem /BRCosmetics/Accessories/FORT_Backpacks/M_MED_Snowman_Fashion_Backpack/FX/NS_Backpack_SnowmanFashion_Idle
dependencies {
 usedDataInterfaces [&NiagaraDataInterfaceCurve, &NiagaraDataInterfaceColorCurve]
 usedDynamicInputs [&Boolean_AndOperation]
 usedModules [
  &SystemState
  &FN_Integrator
  &self:SystemSpawnScript.SystemScriptSource.SystemScriptGraph.NiagaraNodeAssignment_0.SetVariables_#1
  &self:SystemSpawnScript.SystemScriptSource.SystemScriptGraph.NiagaraNodeAssignment_1.SetVariables_#2
 ]
}
emitters {
 Idle_Flakes {
  emitterSpawnScript [
   /SetVariables(Emitter.RandomSeed:NiagaraInt32=/Niagara/DynamicInputs/Random/RandomInt(Randomness Mode:ENiagaraRandomnessMode=::NewEnumerator0, Random Seed:NiagaraInt32=0, Fixed Random Seed:NiagaraBool=false)) {isSetParameters=true, module=SetVariables_#3}
   /Game/Effects/Niagara/Modules/Partition/FN_PartitionAttributes(Number of Bone Sets=2, Particles Per Partition=8, Partitions=3)
  ]
  emitterUpdateScript [
   &EmitterState(Life Cycle Mode=::NewEnumerator1, Loop Duration=1, UseLoopDelay=false)
   &SpawnBurst_Instantaneous(Spawn Count=$Emitter.FN_PartitionAttributes.Spawn Total, Use Spawn Probability=false) {enabled=false}
   &SpawnRate(SpawnRate=6, Use Spawn Probability=true)
   &SpawnRate(
    SpawnRate=&MakeCustomFloatFromBool(
     Bool=&SetBoolByFloatComparison(
      A=&VectorLength(Vector=&NormalizeVector(Vector To Normalize=$Engine.Owner.Velocity))
      B=0.01
     )
     True Float=0
     False Float=30
    )
    Use Spawn Probability=true
   ) {enabled=false, module=SpawnRate001}
   &SpawnPerUnit(Module.Use Max Movement Threshold=true, Module.Use Movement Tolerance=true, Use Spawn Probability=false) {enabled=false}
  ]
  particleSpawnScript [
   /SetVariables(
    Particles.BoneFlow Height=1
    Particles.Boneflow Radius:NiagaraFloat=&Lerp_Float(A=1, B=1, Alpha=$Particles.BoneFlow Height)
    Particles.RandomRangeAngle:NiagaraFloat=&RandomRangeFloat(Minimum=-0.3, Maximum=0.3)
   ) {isSetParameters=true, module=SetVariables_#4}
   /Game/Effects/Niagara/Modules/Partition/FN_Particle_partition_v5/FN_Particle_Partition_v5(Pattern:ENiagaraPartitionWaveForm=::NewEnumerator1)
   /Niagara/Modules/Spawn/Initialization/InitializeParticle(
    Lifetime=&RandomRangeFloat(Minimum=1.5, Maximum=2)
    Module.Write Color=true
    Module.Write Lifetime=true
    Module.Write Mass=true
    Module.Write Material Random=true
    Module.Write Position=false
    Module.Write Scale=false
    Module.Write SpriteRotation=true
    Module.Write SpriteSize=true
    Position=&SimulationPosition()
    Sprite Rotation=&RandomRangeFloat(Minimum=0, Maximum=1)
    Sprite Size=&Vector2DFromFloat(Value=&RandomRangeFloat(Minimum=2, Maximum=5))
   )
   &Color(Color=<rgba 0.4 0.59 1 1>, Scale Alpha=$Particles.MaterialRandom)
   &FN_BoneFlow(
    Angle:NiagaraFloat=&Lerp_Float(
     A=$Particles.RandomRangeAngle
     B=&Add_Float(A=$Particles.RandomRangeAngle, B=0.5)
     Alpha=&ModuloFloat(A=&MakeFloatFromInt(Int=$Particles.Partition.ID), B=2)
    )
    Height:NiagaraFloat=&Max_Float(A:NiagaraFloat=$Particles.MaterialRandom, B:NiagaraFloat=0.001)
    Interpolate Mesh Sampling=false
    Radius:NiagaraFloat=$Particles.Boneflow Radius
    SkeletalMesh:NiagaraDataInterfaceSkeletalMesh=$User.SkeletalMesh
    Write Position Attribute=true
   )
   &InheritVelocity(Module.Limit Speed=true, Velocity Scale=&VectorFromFloat(Value=5)) {enabled=false}
   &AddVelocity(Velocity=&Multiply_VectorByFloat(Vector=$Output.FN_BoneFlow.SplineTangent, Float=-5))
   &BoxLocation(Box Offset=<0 -8 -13>, Box Size=<0 0 0>, MaskSpawnGroup=false, Module.Override Seed=false, Rotate Box=false, Surface Only=false)
   &AddVelocity(Velocity=&Multiply_VectorByFloat(Vector=$Output.FN_BoneFlow.RotatedBiNormal, Float=5)) {enabled=false, module=AddVelocity001}
  ]
  particleUpdateScript [
   /SetVariables(Particles.Boneflow Radius:NiagaraFloat=&Multiply_Float(
     A=$Particles.Boneflow Radius
     B=&FloatFromCurve(FloatCurve=DataInterface() {propertyValues {Curve {keys [<at 0 1>,<at 1 15>]}}})
    )) {enabled=false, isSetParameters=true}
   &FN_BoneFlow(
    Angle:NiagaraFloat=&Lerp_Float(
     A=$Particles.RandomRangeAngle
     B=&Add_Float(A=$Particles.RandomRangeAngle, B=0.5)
     Alpha=&ModuloFloat(A=&MakeFloatFromInt(Int=$Particles.Partition.ID), B=2)
    )
    Height:NiagaraFloat=&Max_Float(A:NiagaraFloat=$Particles.MaterialRandom, B:NiagaraFloat=0.001)
    Interpolate Mesh Sampling=false
    Radius:NiagaraFloat=&Multiply_Float(
     A=$Particles.Boneflow Radius
     B=&FloatFromCurve(FloatCurve=DataInterface() {propertyValues {Curve {keys [<at 0 1>,<at 1 2>]}}}, Scale Curve=5)
    )
    SkeletalMesh:NiagaraDataInterfaceSkeletalMesh=$User.SkeletalMesh
    Write Position Attribute=true
   ) {enabled=false, module=FN_BoneFlow001}
   &AccelerationForce(Acceleration=&RandomRangeVector(Minimum=<0 0 -10>, Maximum=<1 1 -50>))
   &Drag(
    Drag=&RandomRangeFloat(Minimum=2, Maximum=5)
    Rotational Drag=20
    Use Linear Drag=true
    Use Rotational Drag=false
   )
   &SolveForcesAndVelocity(Clamp Velocity=false, Limit Acceleration=false)
   &ParticleState()
   &SpriteRotationRate(Rotation Rate=&FloatFromCurve(
     FloatCurve=DataInterface() {propertyValues {Curve {keys [<at 0 0 RCIM_Cubic>,<at 0.3 1 RCIM_Cubic>,<at 1 0>]}}}
     Scale Curve=&RandomRangeFloat(Minimum=-100, Maximum=100)
    ))
   &ScaleColor(
    Scale Alpha=&FloatFromCurve(FloatCurve=DataInterface() {propertyValues {Curve {keys [<at 0.6 1>,<at 1 0>]}}})
    ScaleA=true
    ScaleRGB=true
   )
   &ScaleSpriteSize(Scale Factor=&Vector2DFromFloat(Value=&FloatFromCurve(FloatCurve=DataInterface() {propertyValues {Curve {keys [
          <at 0 0 RCIM_Cubic>
          {arriveTangent=2, interpMode=RCIM_Cubic, leaveTangent=2, time=0.283, value=0.995}
          {arriveTangent=-1.3877, interpMode=RCIM_Cubic, leaveTangent=-1.3877, time=0.5, value=1}
          <at 1 0 RCIM_Cubic>
         ]}}})))
   &ScaleColor(
    Scale Alpha=&FloatFromCurve(
     FloatCurve=$System.ScaleDistanceFadeCurve
     CurveIndex=&Multiply_Float(
      A=&VectorLength(Vector=&Subtract_Vector(
        A=Linked() {linkedVariable {name=Particles.Position, type=NiagaraPosition}}
        B=&Add_Vector(A=Linked() {linkedVariable {name=Engine.Owner.Position, type=NiagaraPosition}}, B=$System.ScaleDistanceOffset)
       ))
      B=0.003
     )
     Scale Curve=0.6
    )
    ScaleA=true
    ScaleRGB=true
   ) {module=ScaleColor002}
  ]
  rendererClasses [&NiagaraSpriteRendererProperties]
  renderers [{rendererClass=&NiagaraSpriteRendererProperties}]
 }
 Idle_Puffs {
  emitterSpawnScript [
   /SetVariables(Emitter.RandomSeed:NiagaraInt32=&RandomInt(Randomness Mode:ENiagaraRandomnessMode=::NewEnumerator0, Random Seed:NiagaraInt32=0, Fixed Random Seed:NiagaraBool=false)) {isSetParameters=true, module=SetVariables_#3}
   &FN_PartitionAttributes(Number of Bone Sets=2, Particles Per Partition=8, Partitions=3)
  ]
  emitterUpdateScript [
   &EmitterState(Life Cycle Mode=::NewEnumerator1, Loop Duration=1, UseLoopDelay=false)
   &SpawnRate(
    Spawn Probability=&MakeCustomFloatFromBool(
     Bool=&SetBoolByFloatComparison(A=&VectorLength(Vector=$Engine.Owner.Velocity), B=1)
     True Float=0
     False Float=1
    )
    SpawnRate=15
    Use Spawn Probability=true
   ) {module=SpawnRate002}
  ]
  particleSpawnScript [
   /SetVariables(
    Particles.BoneFlow Height=1
    Particles.Boneflow Radius:NiagaraFloat=&RandomRangeFloat(Minimum=8, Maximum=5)
    Particles.RandomRangeAngle:NiagaraFloat=&RandomRangeFloat(Minimum=-0.3, Maximum=0.3)
   ) {isSetParameters=true, module=SetVariables_#4}
   &FN_Particle_Partition_v5(Pattern:ENiagaraPartitionWaveForm=::NewEnumerator1)
   &InitializeParticle(
    Color=&ColorFromCurve(
     ColorCurve=DataInterface() {propertyValues {
       AlphaCurve {keys [<at 0 0.1 RCIM_Cubic>,<at 1 0.05 RCIM_Cubic>]}
       BlueCurve {keys [<at 0 1 RCIM_Cubic>,<at 1 1 RCIM_Cubic>]}
       GreenCurve {keys [<at 0 1 RCIM_Cubic>,<at 1 0.8149 RCIM_Cubic>]}
       RedCurve {keys [<at 0 1 RCIM_Cubic>,<at 1 0.8 RCIM_Cubic>]}
      }}
     CurveIndex=$Particles.Partition.NormalizedID
    )
    Lifetime=&RandomRangeFloat(Minimum=1.4, Maximum=2)
    Module.Write Color=true
    Module.Write Lifetime=true
    Module.Write Mass=true
    Module.Write Material Random=true
    Module.Write Position=false
    Module.Write Scale=false
    Module.Write SpriteRotation=true
    Module.Write SpriteSize=true
    Position=&SimulationPosition()
    Sprite Rotation=&RandomRangeFloat(Minimum=0, Maximum=360)
    Sprite Size=&Vector2DFromFloat(Value=&RandomRangeFloat(Minimum=10, Maximum=10))
   )
   &FN_BoneFlow(
    Angle:NiagaraFloat=&Lerp_Float(
     A=$Particles.RandomRangeAngle
     B=&Add_Float(A=$Particles.RandomRangeAngle, B=0.5)
     Alpha=&ModuloFloat(A=&MakeFloatFromInt(Int=$Particles.Partition.ID), B=2)
    )
    Height:NiagaraFloat=&Max_Float(A:NiagaraFloat=$Particles.MaterialRandom, B:NiagaraFloat=0.001)
    Interpolate Mesh Sampling=false
    Radius:NiagaraFloat=$Particles.Boneflow Radius
    SkeletalMesh:NiagaraDataInterfaceSkeletalMesh=$User.SkeletalMesh
    Write Position Attribute=true
   )
   &AddVelocity(Velocity=&Multiply_VectorByFloat(Vector=$Output.FN_BoneFlow.SplineTangent, Float=-5))
   &BoxLocation(Box Offset=<0 -5 -5>, Box Size=<0 0 0>, MaskSpawnGroup=false, Module.Override Seed=false, Rotate Box=false, Surface Only=false)
  ]
  particleUpdateScript [
   &AccelerationForce(Acceleration=&RandomRangeVector(Minimum=<0 0 0>, Maximum=<1 1 -100>))
   &InheritVelocity(Module.Limit Speed=true, Velocity Limit=200, Velocity Scale=&VectorFromFloat(Value=-1)) {enabled=false}
   &Drag(
    Drag=&RandomRangeFloat(Minimum=5, Maximum=5)
    Rotational Drag=20
    Use Linear Drag=true
    Use Rotational Drag=false
   )
   &SolveForcesAndVelocity(Clamp Velocity=false, Limit Acceleration=false)
   &ParticleState()
   /Niagara/Modules/Update/SubUV/SubUVAnimation(Number Of Frames=&Lerp_Float(A=50, B=50, Alpha=$Particles.NormalizedAge)) {enabled=false}
   /Niagara/Modules/Update/SubUV/V2/SubUVAnimation(
    Sub UV Play Rate:NiagaraFloat=&Multiply_Float(A=0.3, B=$Particles.NormalizedAge)
    SubUV Animation Mode:ENiagara_SubUVLookupMode=::NewEnumerator3
   ) {module=SubUVAnimation001}
   &SpriteRotationRate(Rotation Rate=&FloatFromCurve(
     FloatCurve=DataInterface() {propertyValues {Curve {keys [<at 0 0 RCIM_Cubic>,<at 0.3 1 RCIM_Cubic>,<at 1 0>]}}}
     Scale Curve=&RandomRangeFloat(Minimum=-20, Maximum=20)
    ))
   &ScaleSpriteSize(Scale Factor=&Vector2DFromFloat(Value=&FloatFromCurve(
      FloatCurve=DataInterface() {propertyValues {Curve {keys [
          {arriveTangent=0.7564, interpMode=RCIM_Cubic, leaveTangent=0.7564, tangentMode=RCTM_User, time=-0.0024, value=0.2}
          {arriveTangent=1.4669, interpMode=RCIM_Cubic, leaveTangent=1.4669, time=0.4362, value=1.0444}
          <at 0.9951 1.6633 RCIM_Cubic>
         ]}}}
      Scale Curve=2
     )))
   &ScaleColor(
    Scale Mode=::NewEnumerator1
    Scale RGBA=&MakeVector4_FromLinearColor(LinearColor=&ColorFromCurve(ColorCurve=DataInterface() {propertyValues {
        AlphaCurve {keys [
          <at 0 0 RCIM_Cubic>
          {arriveTangent=1.5193, interpMode=RCIM_Cubic, leaveTangent=1.5193, time=0.2235, value=0.9746}
          {arriveTangent=-1.2368, interpMode=RCIM_Cubic, leaveTangent=-1.2368, time=0.5352, value=0.8131}
          <at 1.0115 0>
          <at 1.094 0>
         ]}
        BlueCurve {keys [<at 0 1 RCIM_Cubic>,<at 1 1 RCIM_Cubic>]}
        GreenCurve {keys [<at 0 0.715 RCIM_Cubic>,<at 1 0.79 RCIM_Cubic>]}
        RedCurve {keys [<at 0 0.62 RCIM_Cubic>,<at 1 0.7 RCIM_Cubic>]}
       }}))
    ScaleRGBA=true
   ) {module=ScaleColor001}
   /SetVariables(Particles.Lifetime=&Lerp_Float(
     A=$Particles.Initial.Lifetime
     B=2
     Alpha=&VectorLength(Vector=&Multiply_VectorByFloat(Vector=$Engine.Owner.Velocity, Float=1))
    )) {enabled=false, isSetParameters=true}
   /Niagara/Modules/Update/Color/ScaleColorBySpeed(A Scale Min=0.2, Max Speed Threshold=150, Min Speed Threshold=10, RGB Scale Min=1, Source Velocity:Vector3f=$Engine.Owner.Velocity) {enabled=false}
   /SetVariables(
    Particles.Alpha:NiagaraFloat=&MakeFloatFromVector4(Vector 4=&MakeVector4_FromLinearColor(LinearColor=$Particles.Color), Channel=::NewEnumerator3)
    Particles.Color RGB:Vector3f=&MakeVectorFromLinearColorRGB(Color=$Particles.Color)
   ) {enabled=false, isSetParameters=true}
   /SetVariables(Particles.Color=&MakeLinearColorFromVectorAndFloat(
     "Vector (RGB)"=$Particles.Color RGB
     "Float (Alpha)"=&MakeCustomFloatFromBool(
      Bool=&SetBoolByFloatComparison(A=$Particles.Alpha, B=0.2, Comparison Type=::NewEnumerator6)
      True Float=0
      False Float=1
     )
    )) {enabled=false, isSetParameters=true}
   &ScaleColor(
    Scale Alpha=&FloatFromCurve(
     FloatCurve=$System.ScaleDistanceFadeCurve
     CurveIndex=&Multiply_Float(
      A=&VectorLength(Vector=&Subtract_Vector(
        A=Linked() {linkedVariable {name=Particles.Position, type=NiagaraPosition}}
        B=&Add_Vector(A=Linked() {linkedVariable {name=Engine.Owner.Position, type=NiagaraPosition}}, B=$System.ScaleDistanceOffset)
       ))
      B=0.008
     )
     Scale Curve=2
    )
    ScaleA=true
    ScaleRGB=true
   ) {module=ScaleColor002}
  ]
  rendererClasses [&NiagaraSpriteRendererProperties]
  renderers [{rendererClass=&NiagaraSpriteRendererProperties}]
 }
}
summary {
 systemName=NS_Backpack_SnowmanFashion_Idle
 emitters [
  {emitterName=Idle_Flakes, rendererClasses [&NiagaraSpriteRendererProperties]}
  {emitterName=Idle_Puffs, rendererClasses [&NiagaraSpriteRendererProperties]}
 ]
 userVariables [
  User.BlackSmokeTrailSpawnRate {type=NiagaraFloat, defaultValue=10, description=""}
  User.ColorChange {type=LinearColor, description="", defaultValue {struct=LinearColor, value=<rgba 1 1 1 1>}}
  User.EyeGlowMultiplier {type=NiagaraFloat, defaultValue=0, description=""}
  User.EyeGlowStrength {type=NiagaraFloat, defaultValue=0, description=""}
  User.JumpBurst {type=NiagaraFloat, defaultValue=0, description=""}
  User.JumpTimeRange {type=NiagaraFloat, defaultValue=0, description=""}
  User.LandBurst {type=NiagaraFloat, defaultValue=0, description=""}
  User.LightStrength {type=NiagaraFloat, defaultValue=0, description=""}
  User.Right2ImpactVelocity {type=Vector3f, description="", defaultValue {struct=Vector3f, value=<0 0 0>}}
  User.SkeletalMesh {
   type=NiagaraDataInterfaceSkeletalMesh
   defaultValue {
    struct=NiagaraExt_VariableValue_DataInterface
    value {dataInterface=&self:NiagaraDataInterfaceSkeletalMesh_1}
   }
  }
 ]
}
